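# Surface material for the Gothic vehicle wheel cosmetic, built on MF_WheelParent_Simple with an added rim emissive.
# Emissive is the rim diffuse texture raised to Color_Strength, scaled by the rim mask's blue channel and Emissive_Strength, and faded out by red.
# Vertex color red blends from this emissive version back to the parent's attributes.
# The result passes through the shared vehicle damage and de-rez transition functions.
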Material /VehicleCosmetics/Wheels/Gothic/Materials/M_Wheel_Gothic
graph {
 properties {
  bAutomaticallySetUsageInEditor=false
  bUseMaterialAttributes=true
  bUsedWithMeshDeformer=true
  bUsedWithSkeletalMesh=true
  usedShadingModels=Default Lit
 }
 nodes [
  n1 = MaterialFunctionCall() {materialFunction=&MF_WheelParent_Simple}
  n2 = Multiply(
   A=ComponentMask(None=@n1.RimRGBAMask) {b=true}
   B=ScalarParameter() {defaultValue=1, group=4_Surface Defaults, parameterName=Emissive_Strength}
  )
  n3 = Power(
   Base=TextureSampleParameter2D() {parameterName=Rim_DiffuseTexture, texture=/VehicleCosmetics/Textures/T_White}.RGB
   Exp=ScalarParameter() {defaultValue=1, group=4_Surface Defaults, parameterName=Color_Strength}
  )
  n4 = SetMaterialAttributes(
   MaterialAttributes=@n1.MaterialAttributes
   Emissive Color=LinearInterpolate(A=Multiply(A=@n2, B=@n3), Alpha=ComponentMask(None=@n1.RimRGBAMask) {r=true}) {constB=0}
  ) {attributeSetTypes [#1]}
  MaterialFunctionCall(In=MaterialFunctionCall(Material=BlendMaterialAttributes(A=@n4, B=@n1.MaterialAttributes, Alpha=@n1.VertexColorR)) {materialFunction=&MF_BaseVehicle_Damage}.Result) {materialFunction=&MF_VehicleDeRez_Transition}
 ]
}
